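# Additive, unlit, two-sided mesh-particle material for the helix trail on the Cattus beam effect in the CattusDoggus event.
# Two panning noise textures and a sine edge profile drive a purple curve-atlas color, hue-shifted and tinted by particle color.
# A mask panned by dynamic-parameter animation shapes erodible opacity, softened by depth and view-angle fades outside mobile.
# World position offset combines SplineThicken width, optionally particle-driven or camera-faded, with sine-weighted panning normal-map distortion.
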Material /Figment_S09/Effects/Fort_Effects/Events/CattusDoggus/Cattus/M_Cattus_Beam_HelixTrail
graph {
 properties {
  bUsedWithMeshParticles=true
  blendMode=BLEND_Additive
  pixelDepthOffsetMode=PDOM_Legacy
  shadingModel=MSM_Unlit
  twoSided=true
  usedShadingModels=Unlit
 }
 nodes [
  n1 = DynamicParameter() {paramNames [PanAnim, SplineWidthBase, SplineWidthTop, Erode]}
  n2 = StaticSwitchParameter(True=@n1.SplineWidthBase, False=ScalarParameter() {defaultValue=1, parameterName=WidthBase}) {parameterName=UseDPtoHandleSplineWidth}
  n3 = MaterialFunctionCall(Distance=ScalarParameter() {defaultValue=3000, parameterName=WPO_NearCameraFade}) {materialFunction=&Near_Camera_Fade}
  n4 = LinearInterpolate(B=@n2, Alpha=@n3.Result)
  n5 = StaticSwitchParameter(True=@n1.SplineWidthTop, False=ScalarParameter() {defaultValue=1, parameterName=WidthTip}) {parameterName=UseDPtoHandleSplineWidth}
  n6 = LinearInterpolate(B=@n5, Alpha=@n3.Result)
  n7 = VectorParameter() {parameterName=NoiseTile, defaultValue=<rgba 1.5 1 -1 0>}
  n8 = Multiply(A=TextureCoordinate(), B=ComponentMask(None=@n7.RGB) {g=true, r=true})
  n9 = Time()
  n10 = DynamicParameter() {paramNames [PanAnim, SplineWidthBase, SplineWidthTop, Erode]}
  n11 = Panner(
   Coordinate=StaticSwitchParameter(True=MaterialFunctionCall(XY=@n8) {materialFunction=&Swizzle}.YX, False=@n8) {parameterName=Swizzle}
   Time=StaticSwitchParameter(True=@n9, False=Add(A=@n9, B=@n10.SplineWidthBase)) {parameterName=UseDPtoHandleSplineWidth}
   Speed=AppendVector(A=@n7.B, B=@n7.B)
  ) {speedX=-1}
  n12 = Multiply(A=TextureSampleParameter2D(UVs=@n11) {parameterName=WPO, samplerType=SAMPLERTYPE_Normal, texture=/Figment_S09/Effects/Fort_Effects/Textures/T_Weap_Distort_N_02}.RGB) {constB=1.1}
  n13 = TextureSample(UVs=@n11) {samplerType=SAMPLERTYPE_LinearColor, texture=/Game/Effects/Fort_Effects/Textures/Tile/LinerNoise64x64}
  n14 = Constant() {r=0.7}
  n15 = StaticSwitchParameter(
   True=MaterialFunctionCall(
    BaseNormal=@n12
    AdditionalNormal=MaterialFunctionCall(Normal=@n13.RGB, Flatness=@n14) {materialFunction=&FlattenNormal}.Result
   ) {materialFunction=&BlendAngleCorrectedNormals}.Result
   False=@n12
  ) {parameterName=UseDetailNoiseWPO}
  n16 = TextureCoordinate()
  n17 = StaticSwitchParameter(True=MaterialFunctionCall(XY=@n16) {materialFunction=&Swizzle}.YX, False=@n16) {parameterName=Swizzle}
  n18 = Sine(None=Multiply(A=MaterialFunctionCall(Float2=@n17) {materialFunction=&BreakOutFloat2Components}.R) {constB=0.5})
  Add(
   A=MaterialFunctionCall(
    WidthBase=StaticSwitchParameter(True=@n4, False=@n2) {parameterName=UseCameraFade}
    WidthTip=StaticSwitchParameter(True=@n6, False=@n5) {parameterName=UseCameraFade}
    Expand U or V UV Channel=StaticBoolParameter() {defaultValue=true, parameterName=Expand U or V UV Channel}
   ) {materialFunction=&SplineThicken}.WorldPositionOffset
   B=Multiply(A=Multiply(A=@n15, B=@n18), B=ScalarParameter() {defaultValue=200, parameterName=WPO}) {constB=200}
  )
  n19 = ParticleColor()
  n20 = TextureCoordinate()
  n21 = MaterialFunctionCall(Float2=StaticSwitchParameter(True=MaterialFunctionCall(XYZ=@n20) {materialFunction=&Swizzle}.YX, False=@n20) {parameterName=Swizzle}) {materialFunction=&BreakOutFloat2Components}
  n22 = Multiply(A=@n21.G) {constB=0.5}
  n23 = VectorParameter() {parameterName=EmissiveTile, defaultValue=<rgba 3 1 -2.54 0>}
  n24 = Multiply(A=TextureCoordinate(), B=ComponentMask(None=@n23.RGB) {g=true, r=true})
  n25 = MaterialFunctionCall(Float2=StaticSwitchParameter(True=MaterialFunctionCall(XY=@n24) {materialFunction=&Swizzle}.YX, False=@n24) {parameterName=Swizzle}) {materialFunction=&BreakOutFloat2Components}
  n26 = AppendVector(A=Power(Base=@n25.R) {constExponent=0.6}, B=@n25.G)
  n27 = Time()
  n28 = StaticSwitchParameter(True=@n27, False=Add(A=@n27, B=DynamicParameter() {paramNames [PanAnim, SplineWidthBase, SplineWidthTop, Erode]}.SplineWidthBase)) {parameterName=UseDPtoHandleSplineWidth}
  n29 = AppendVector(A=@n23.B, B=@n23.B)
  n30 = Panner(
   Coordinate=@n26
   Time=@n28
   Speed=Add(A=@n29, B=ScalarParameter() {defaultValue=-0.46, parameterName=EmissivePanningOffsetBetweenAandB}) {constB=-0.46}
  ) {speedX=-3}
  n31 = Multiply(
   A=TextureSampleParameter2D(UVs=Panner(Coordinate=@n26, Time=@n28, Speed=@n29) {speedX=-2.54}) {parameterName=EmissiveA, texture=/Game/Effects/Fort_Effects/Gameplay/SmokeDustExplosions/Textures/T_NoisePack_01}.R
   B=Multiply(A=TextureSampleParameter2D(UVs=@n30) {parameterName=EmissiveB, texture=/Game/Effects/Fort_Effects/Textures/T_Smooth_tile}.R) {constB=2}
  )
  n32 = Add(
   A=Power(Base=Sine(None=@n22))
   B=LinearInterpolate(
    A=ScalarParameter() {defaultValue=-0.02, parameterName=EmissiveDetailLowRange}
    B=ScalarParameter() {defaultValue=0.6, parameterName=EmissiveDetailHighRange}
    Alpha=Divide(A=@n31) {constB=15}
   ) {constA=-0.02, constB=0.6}
  )
  n33 = Multiply(A=@n21.R) {constB=0.5}
  n34 = VectorParameter() {parameterName=MaskTile, defaultValue=<rgba -1.5 1 1 0>}
  n35 = Multiply(A=TextureCoordinate(), B=ComponentMask(None=@n34.RGB) {g=true, r=true})
  n36 = MaterialFunctionCall(XY=@n35) {materialFunction=&Swizzle}
  n37 = TextureSampleParameter2D(UVs=Panner(
    Coordinate=StaticSwitchParameter(True=@n36.YX, False=@n35) {parameterName=Swizzle}
    Time=LinearInterpolate(Alpha=@n1.PanAnim) {constB=4}
    Speed=AppendVector(A=@n34.B, B=@n34.B)
   ) {speedX=1}) {
   parameterName=Mask
   samplerSource=SSM_Clamp_WorldGroupSettings
   samplerType=SAMPLERTYPE_LinearColor
   texture=/Game/Effects/Fort_Effects/Textures/HeightForWater
  }
  n38 = VectorParameter() {parameterName=MaskDetailTile, defaultValue=<rgba 1 1 0 0>}
  n39 = Multiply(A=TextureCoordinate(), B=AppendVector(A=@n38.R, B=@n38.R))
  n40 = MaterialFunctionCall(XY=@n39) {materialFunction=&Swizzle}
  n41 = TextureSampleParameter2D(UVs=Panner(Coordinate=StaticSwitchParameter(True=@n40.YX, False=@n39) {parameterName=Swizzle}, Speed=AppendVector(A=@n38.B, B=@n38.B))) {parameterName=MaskDetail, texture=/Game/Effects/Fort_Effects/Textures/Energy/T_OutpostShockwaveStreaks_02}
  n42 = Multiply(
   A=Multiply(A=@n32, B=Sine(None=@n33))
   B=StaticSwitchParameter(True=Multiply(A=@n37.R, B=@n41.R), False=@n37.R) {parameterName=UseMaskDetail}
  )
  n43 = Divide(A=@n1.Erode)
  n44 = Multiply(A=@n19.A, B=Saturate(None=Subtract(A=@n42, B=@n43)))
  n45 = Multiply(
   A=DepthFade(Opacity=@n44, FadeDistance=ScalarParameter() {parameterName=FadeDistance})
   B=MaterialFunctionCall("FadeSensitivity (optional)"=ScalarParameter() {parameterName=FadeBasedOnViewAngle}) {materialFunction=&FadeBasedOnViewAngle}.Result
  )
  FeatureLevelSwitch(Default=@n45, ES2=@n44, ES3_1=@n44, SM4_REMOVED=@n45)
  n46 = CurveAtlasRowParameter(CurveTime=@n32) {
   atlas=&AthenaCurveAtlas01
   curve=/Game/Effects/Fort_Effects/Events/CattusDoggus/Cattus/Curve_Cattus_Purple
   defaultValue=104
  }
  Multiply(
   A=MaterialFunctionCall(Hue Shift Percentage=ScalarParameter() {parameterName=HueShfit}, Texture=@n46) {materialFunction=&HueShift}.Result
   B=@n19.RGB
  )
 ]
}
